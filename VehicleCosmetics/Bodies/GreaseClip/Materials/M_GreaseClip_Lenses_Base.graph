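# Translucent thin-translucent lens material for the GreaseClip vehicle body, with an added Mr. Fusion glow.
# MF_Lenses_Base_v2 supplies the base attributes and thin-translucent output, with static switches for custom lenses, shared headlights and translucency sorting.
# A tinted inverse-Fresnel emissive, masked to pre-skinned local height above a fixed threshold via a vertex interpolator, is added to the lens emissive.
# The combined attributes then pass through MF_Vehicle_FNPostFX.
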Material /VehicleCosmetics/Bodies/GreaseClip/Materials/M_GreaseClip_Lenses_Base
graph {
 properties {
  allowTranslucentCustomDepthWrites=true
  bAutomaticallySetUsageInEditor=false
  bUseMaterialAttributes=true
  bUsedWithMeshDeformer=true
  bUsedWithSkeletalMesh=true
  blendMode=BLEND_Translucent
  previewMesh=/VehicleCosmetics/Bodies/GreyCar/Meshes/SK_Body_GreyCar
  refractionMethod=RM_IndexOfRefraction
  shadingModel=MSM_ThinTranslucent
  translucencyLightingMode=TLM_SurfacePerPixelLighting
  usedShadingModels=Thin Translucent
 }
 nodes [
  MaterialFunctionCall(In=MaterialFunctionCall() {materialFunction=&MF_BaseVehicle_Damage}.Result) {materialFunction=&MF_VehicleDeRez_Transition}
  n1 = MaterialFunctionCall!(
   UseCustomRearlights=StaticBoolParameter() {group=Lenses, parameterName=UseCustomRearlights}
   UseCustomReflector=StaticBoolParameter() {group=Lenses, parameterName=UseCustomReflectors}
   UseCustomSpecial=StaticBoolParameter() {group=Lenses, parameterName=UseCustomMiscellaneous}
   UseSortTranslucency=StaticBoolParameter() {group=Lenses, parameterName=UseSortTranslucency}
   UseHeadlightsAsShared=StaticBoolParameter() {defaultValue=true, group=Lenses, parameterName=UseHeadlightsAsShared}
  ) {materialFunction=/VehicleCosmetics/Materials/VehicleParents_Version2/Material_Functions/MF_Lenses_Base_v2}
  n2 = GetMaterialAttributes(None=@n1.Result) {attributeGetTypes [#1]}
  n3 = VertexInterpolator(VS=LocalPosition() {localOrigin=InstancePreSkinning}.Z)
  n4 = VectorParameter() {group=Mr. Fusion, parameterName=MrFusion_Color, defaultValue=<rgba 1 0.5269 0 1>}
  n5 = OneMinus(None=Fresnel(ExponentIn=ScalarParameter() {defaultValue=0.75, group=Mr. Fusion, parameterName=MrFusion_FresnelExp}) {baseReflectFraction=0})
  n6 = Multiply(
   A=OneMinus(None=Step(Y=@n3.PS) {constX=145})
   B=Multiply(A=Multiply(A=@n4.RGB, B=@n5), B=ScalarParameter() {defaultValue=3, group=Mr. Fusion, parameterName=MrFusion_EmissiveIntensity})
  )
  MaterialFunctionCall!(In=SetMaterialAttributes(MaterialAttributes=@n2.MaterialAttributes, Emissive Color=Add(A=@n2.Emissive Color, B=@n6)) {attributeSetTypes [#1]}) {materialFunction=/VehicleCosmetics/Materials/MaterialFunctions/MF_Vehicle_FNPostFX}
  n7 = GetMaterialAttributes(None=@n1.Result) {attributeGetTypes [#2, #3]}
  ThinTranslucentMaterialOutput!(TransmittanceColor=@n7.TransmittanceColor, SurfaceCoverage=@n7.TransmittanceColor)
 ]
}
